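; BP_SpawnAction_SpawnInAir_PAVBoss is a spawn action that decides whether the PAV boss may spawn in the air.
; CanSpawnInAir passes only when the Battle Royale game phase logic component is valid and its safe zone start-up is StartsWithNoAirCraft.
; The source point provider input is unused, and OnSpawnedInAir has an empty body.
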(fn CanSpawnInAir
  (params (SpawnedActor :type Actor%20Object%20Reference) (SourcePointProvider :type Fort%20Athena%20Living%20World%20Point%20Provider%20Interface%20Interface%20%28by%20ref%29))
  (return (and (Utilities|IsValid (BattleRoyaleGamePhaseLogic|GetBattleRoyaleGamePhaseLogicComponent SpawnedActor) :!class /Script/Engine.KismetSystemLibrary) (Utilities|Enum|Equal(Enum) (BattleRoyaleGamePhaseLogic|GetSafeZoneStartUp (BattleRoyaleGamePhaseLogic|GetBattleRoyaleGamePhaseLogicComponent SpawnedActor)) "StartsWithNoAirCraft"))))

(fn OnSpawnedInAir
  (params (SpawnedActor :type Actor%20Object%20Reference) (PositionInAir :type Vector%20%28by%20ref%29 :default "0, 0, 0"))
  (return))
